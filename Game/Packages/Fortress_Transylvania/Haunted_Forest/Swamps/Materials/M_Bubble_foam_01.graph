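# Translucent lit mesh-particle material for animated bubbly foam in the Haunted Forest swamps.
# Two foam color and normal layers pan at differing speeds, their UVs distorted by a slower panning texture.
# A V-gradient band and particle alpha shape the foam, with depth fade and a dynamic parameter controlling opacity.
# Particle color tints it, and a fixed-angle view-space normal dot product adds a fake specular glint.
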Material /Game/Packages/Fortress_Transylvania/Haunted_Forest/Swamps/Materials/M_Bubble_foam_01.M_bubble_foam_01
graph {
 properties {
  bTangentSpaceNormal=false
  bUsedWithMeshParticles=true
  blendMode=BLEND_Translucent
  numCustomizedUVs=1
  opacityMaskClipValue=0.999
  pixelDepthOffsetMode=PDOM_Legacy
  refractionMethod=RM_IndexOfRefraction
  usedShadingModels=Default Lit
 }
 nodes [
  n1 = Constant3Vector() {constant=<rgba 0 0 1 0>}
  n2 = TextureSample(UVs=Panner(Coordinate=TextureCoordinate() {uTiling=0.5, vTiling=0.5}) {speedY=0.05}) {texture=&2ndInfectedMapLinear}
  n3 = Multiply(A=ComponentMask(None=@n2.R) {g=true, r=true}) {constB=0.1}
  n4 = Panner(Coordinate=Add(A=@n3, B=TextureCoordinate())) {speedX=0.08, speedY=0.2}
  n5 = Panner(Coordinate=Add(A=@n3, B=TextureCoordinate())) {speedX=-0.05, speedY=0.2}
  n6 = Normalize(VectorInput=Add(
    A=TextureSample(UVs=@n4) {samplerType=SAMPLERTYPE_Normal, texture=/Game/Packages/Fortress_Transylvania/Haunted_Forest/Swamps/Textures/T_sea_foam_03_n}.RGB
    B=TextureSample(UVs=@n5) {samplerType=SAMPLERTYPE_Normal, texture=&T_sea_foam_03_n}.RGB
   ))
  n7 = Multiply(A=@n6, B=Constant3Vector() {constant=<rgba 4 4 1 0>})
  n8 = ParticleColor()
  n9 = MaterialFunctionCall() {materialFunction=&LinearGradient}
  n10 = Power(Base=@n9.VGradient) {constExponent=3}
  n11 = Power(Base=OneMinus(None=@n9.VGradient)) {constExponent=1.2}
  n12 = Multiply(A=Clamp(None=Multiply(A=@n10) {constB=5}), B=Clamp(None=Multiply(A=@n11) {constB=2}))
  n13 = Power(Base=Clamp(None=OneMinus(None=@n12)), Exp=LinearInterpolate(Alpha=@n8.A) {constB=200}) {constExponent=35}
  n14 = LinearInterpolate(
   A=@n1
   B=LinearInterpolate(A=LinearInterpolate(A=@n1, B=@n7, Alpha=@n8.A), B=@n1, Alpha=Clamp(None=@n13)) {constA=1}
   Alpha=DepthFade() {fadeDistanceDefault=10}
  )
  Add(A=Multiply(A=Clamp(None=@n14)) {constB=0.3}, B=Constant() {r=1})
  n15 = Clamp(None=Multiply(A=Power(Base=@n12)) {constB=20})
  n16 = TextureSample(UVs=@n4) {texture=/Game/Packages/Fortress_Transylvania/Haunted_Forest/Swamps/Textures/T_sea_foam_03}
  n17 = TextureSample(UVs=@n5) {texture=&T_sea_foam_03}
  n18 = Add(A=@n16.B, B=Multiply(A=@n17.B) {constB=0.4})
  n19 = Multiply(A=@n8.A, B=Clamp(None=@n12))
  n20 = Power(Base=@n18, Exp=LinearInterpolate(Alpha=@n19) {constA=5, constAlpha=1}) {constExponent=500}
  n21 = LinearInterpolate(Alpha=@n19) {constAlpha=0, constB=2}
  n22 = Multiply(A=Add(A=@n16.A, B=Multiply(A=@n17.A) {constB=0.1}), B=Clamp(None=Multiply(A=@n20, B=@n21) {constB=2}))
  n23 = Clamp(None=Add(A=@n15, B=Multiply(A=@n22) {constB=50}))
  DepthFade(Opacity=LinearInterpolate(A=@n23, B=Multiply(A=@n23) {constB=0}, Alpha=DynamicParameter() {defaultValue=<rgba 0 0 0 1>}.R) {constA=1, constAlpha=0, constB=0}) {fadeDistanceDefault=10}
  n24 = Transform(None=@n6) {transformType=TRANSFORM_View}
  n25 = Multiply(
   A=DotProduct(A=Normalize(VectorInput=@n24), B=Constant3Vector() {desc=Control Light angle, constant=<rgba 0 -0.86 0.5 0>})
   B=Constant() {r=1}
  )
  n26 = Clamp(None=Multiply(A=Power(Base=@n25) {constExponent=60}) {constB=2})
  n27 = Multiply(A=@n15, B=Clamp(None=Add(A=@n2.R) {constB=0.3}))
  n28 = LinearInterpolate(A=Add(A=@n16.G, B=@n17.G), B=@n16.G, Alpha=@n16.G) {constA=5, constAlpha=1}
  n29 = Clamp(None=Multiply(A=Add(A=@n16.A, B=@n17.A)) {constB=0.5})
  n30 = Add(A=@n18) {constB=0.7}
  n31 = Multiply(A=@n8.RGB, B=Multiply(A=LinearInterpolate(A=@n27, B=@n28, Alpha=@n29), B=Clamp(None=@n30)))
  MaterialFunctionCall(DiffuseColor=Add(A=Multiply(A=@n26, B=@n18), B=Clamp(None=@n31))) {materialFunction=&ConvertFromDiffSpec}
 ]
}
