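# Translucent, lit stylized fog cloud base material for sprite and mesh particles.
# Rotating flow maps perturb two panning distortion samples that warp a dust cloud texture and noise map, with dynamic parameters controlling distortion, speed and emissive.
# The noise shades particle-colored emissive and drives vertex offset along normals.
# Opacity combines cloud, particle alpha, Fresnel, depth and distance fades; a UseFade switch links opacity and offset to the MPC_Butterfly Blackout value.
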Material /Game/Packages/Fortress_Transylvania/Haunted_Forest/Swamps/Materials/M_Stylized_Fog_Cloud_01_base
graph {
 properties {
  bTangentSpaceNormal=false
  bUsedWithMeshParticles=true
  bUsedWithParticleSprites=true
  blendMode=BLEND_Translucent
  opacityMaskClipValue=0.999
  pixelDepthOffsetMode=PDOM_Legacy
  translucencyLightingMode=TLM_VolumetricPerVertexDirectional
  translucencyPass=MTP_BeforeDOF
  usedShadingModels=Default Lit
 }
 nodes [
  n1 = ParticleColor()
  n2 = DynamicParameter() {defaultValue=<rgba 0.3 0.2 0 1>, paramNames [Distortion Strength, Distortion Strength Normal, Emissive, Panner Speed]}
  n3 = Multiply(A=@n2.Panner Speed, B=Time())
  n4 = TextureSample(UVs=Rotator(Coordinate=TextureCoordinate() {uTiling=2, vTiling=2}) {centerX=1, speed=0.0312}) {samplerType=SAMPLERTYPE_LinearColor, texture=/Game/VisualThreatManager/StormVisuals/Test/SpawnParticles/Streamers/NewFolder/SampleFlow}
  n5 = TextureSample(UVs=Rotator(Coordinate=TextureCoordinate() {uTiling=1.5, vTiling=1.5}) {centerY=1, speed=-0.0312}) {samplerType=SAMPLERTYPE_LinearColor, texture=&SampleFlow}
  n6 = ConstantBiasScale(None=ComponentMask(None=Add(A=@n4.RGB, B=@n5.RGB)) {g=true, r=true}) {bias=-0.5, scale=0.3}
  n7 = Add(A=Panner(Coordinate=TextureCoordinate() {uTiling=0.35, vTiling=0.3}, Time=@n3) {bFractionalPart=true, speedX=0.02, speedY=-0.015}, B=@n6)
  n8 = Add(A=Panner(Coordinate=TextureCoordinate() {uTiling=0.5, vTiling=0.4}, Time=@n3) {bFractionalPart=true, speedX=-0.02, speedY=0.005}, B=@n6)
  n9 = ComponentMask(None=AppendVector(
    A=TextureSample(UVs=@n7) {texture=/Game/Effects/Fort_Cine_Effects/Leaves/Textures/T_Cine_Tilling_Distortion_Map_01}.G
    B=TextureSample(UVs=@n8) {texture=&T_Cine_Tilling_Distortion_Map_01}.G
   )) {g=true, r=true}
  n10 = DynamicParameter() {defaultValue=<rgba 0.3 0.2 0 1>, paramNames [Distortion Strength, Distortion Strength Normal, Emissive, Panner Speed]}
  n11 = Add(A=Multiply(A=@n9, B=@n10.Distortion Strength Normal), B=TextureCoordinate() {uTiling=0.6, vTiling=0.6})
  n12 = TextureSample(UVs=@n11) {texture=&2ndInfectedMapLinear}
  n13 = Add(A=@n9) {constB=-0.15}
  n14 = DynamicParameter() {defaultValue=<rgba 0.3 0.2 0 1>, paramNames [Distortion Strength, Distortion Strength Normal, Emissive, Panner Speed]}
  n15 = TextureSample(UVs=Add(A=Multiply(A=@n13, B=@n14.Distortion Strength), B=TextureCoordinate())) {texture=/Game/Packages/Fortress_Transylvania/Haunted_Forest/Swamps/Textures/T_dust_cloud_01}
  n16 = Multiply(A=LinearInterpolate(Alpha=@n12.R) {constA=0.5}, B=@n15.R)
  n17 = DynamicParameter() {defaultValue=<rgba 0.3 0.2 0 1>, paramNames [Distortion Strength, Distortion Strength Normal, Emissive, Panner Speed]}
  MaterialFunctionCall(Emissive=Multiply(A=Multiply(A=@n1.RGB, B=@n16), B=@n17.Emissive)) {materialFunction=&MF_FX_EmissiveAdjustments}
  n18 = WorldPosition()
  n19 = ObjectPositionWS()
  n20 = Multiply(A=Multiply(A=Add(A=@n12.R) {constB=-0.5}, B=VertexNormalWS()), B=ScalarParameter() {defaultValue=0.5, parameterName=WPO Intensity})
  n21 = OneMinus(None=CollectionParameter() {collection=/Game/Athena/Environments/Nexus/Materials/MPC_Butterfly, parameterName=Blackout})
  StaticSwitchParameter(True=LinearInterpolate(A=Subtract(A=@n18.XYZ, B=@n19), B=@n20, Alpha=Ceil(None=@n21)), False=@n20) {parameterName=UseFade}
  n22 = ParticleColor()
  n23 = Multiply(A=@n15.R, B=@n22.A)
  n24 = MaterialFunctionCall(Exponent=ScalarParameter() {defaultValue=1.5, parameterName=Exponent}, Vertex / Pixel Normal=VertexNormalWS()) {materialFunction=&MF_Fresnel_FX}
  n25 = MaterialFunctionCall(
   Opacity In=DepthFade(Opacity=Multiply(A=@n23, B=@n24.Fresnel), FadeDistance=ScalarParameter() {defaultValue=80, parameterName=Depth Fade Distance}) {fadeDistanceDefault=20}
   Position=WorldPosition().XYZ
   Fade_Start_Distance=ScalarParameter() {defaultValue=200, parameterName=Fade Start Distance}
   Fade_End_Distance=ScalarParameter() {defaultValue=50, parameterName=Fade End Distance}
  ) {materialFunction=&MF_Distance_SyncedFade_FX}
  StaticSwitchParameter(True=Multiply(A=@n25.Result, B=@n21), False=@n25.Result) {parameterName=UseFade}
  Subtract(A=@n15.R, B=OneMinus(None=@n22.A))
  TextureSample(UVs=@n11) {samplerType=SAMPLERTYPE_Normal, texture=/Game/Global/Textures/Default/Tileable_Patterns/2ndInfectedMapNormal}
 ]
}
