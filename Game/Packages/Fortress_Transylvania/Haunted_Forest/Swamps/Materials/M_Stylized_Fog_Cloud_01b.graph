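# Translucent lit particle sprite material for stylized, drifting fog clouds, tinted by particle color.
# Rotating flow samples warp two panning noise maps, which distort the dust cloud texture and erode its opacity, with depth fade modulating the erosion.
# A dynamic parameter offsets the noise pan time.
# Two camera distance blends and a final depth fade shape opacity, while the Fortnite time-of-day collection parameter shifts brightness and opacity between night and day.
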Material /Game/Packages/Fortress_Transylvania/Haunted_Forest/Swamps/Materials/M_Stylized_Fog_Cloud_01b
graph {
 properties {
  bTangentSpaceNormal=false
  bUsedWithParticleSprites=true
  blendMode=BLEND_Translucent
  opacityMaskClipValue=0.999
  pixelDepthOffsetMode=PDOM_Legacy
  translucencyLightingMode=TLM_VolumetricPerVertexNonDirectional
  translucencyPass=MTP_BeforeDOF
  usedShadingModels=Default Lit
 }
 nodes [
  n1 = TextureCoordinate() {uTiling=0.3, vTiling=0.3}
  n2 = LinearInterpolate(A=Multiply(A=DynamicParameter() {defaultValue=<rgba 0 0 0 1>}.Param4) {constB=30}, B=Time())
  n3 = TextureSample(UVs=Rotator(Coordinate=TextureCoordinate() {uTiling=0.5, vTiling=0.5}) {speed=0.0625}) {samplerType=SAMPLERTYPE_LinearColor, texture=/Game/VisualThreatManager/StormVisuals/Test/SpawnParticles/Streamers/NewFolder/SampleFlow}
  n4 = TextureSample(UVs=Rotator(Coordinate=TextureCoordinate() {uTiling=0.5, vTiling=0.5}) {centerX=0.25, centerY=1, speed=-0.0625}) {samplerType=SAMPLERTYPE_LinearColor, texture=&SampleFlow}
  n5 = Multiply(A=Add(A=ComponentMask(None=@n3.RGB) {g=true, r=true}, B=ComponentMask(None=@n4.RGB) {g=true, r=true})) {constB=0.35}
  n6 = TextureSample(UVs=Add(A=Panner(Coordinate=@n1, Time=@n2) {speedX=0.02, speedY=-0.015}, B=@n5)) {texture=/Game/Effects/Fort_Cine_Effects/Leaves/Textures/T_Cine_Tilling_Distortion_Map_01}
  n7 = TextureCoordinate() {uTiling=0.4, vTiling=0.4}
  n8 = TextureSample(UVs=Add(A=Panner(Coordinate=@n7, Time=@n2) {speedX=-0.02, speedY=0.005}, B=@n5)) {texture=&T_Cine_Tilling_Distortion_Map_01}
  n9 = Multiply(A=Clamp(None=Add(A=@n6.G, B=@n8.G))) {constB=3}
  n10 = Multiply(A=ComponentMask(None=@n9) {g=true, r=true}) {constB=0.12}
  n11 = TextureCoordinate()
  n12 = TextureSample(UVs=Panner(Coordinate=Add(A=@n10, B=@n11), Time=Constant() {r=-1.5}) {speedX=0.05, speedY=0.05}) {texture=/Game/Packages/Fortress_Transylvania/Haunted_Forest/Swamps/Textures/T_dust_cloud_01}
  n13 = Clamp(None=OneMinus(None=@n9))
  n14 = ParticleColor()
  n15 = DepthFade() {fadeDistanceDefault=850}
  n16 = Power(Base=@n9, Exp=LinearInterpolate(Alpha=@n15) {constA=3})
  n17 = LinearInterpolate(Alpha=@n15) {constB=30}
  n18 = LinearInterpolate(A=Multiply(A=Multiply(A=@n12.R, B=@n13), B=@n14.A), Alpha=Clamp(None=Multiply(A=@n16, B=@n17) {constB=3})) {constB=0}
  n19 = Constant() {r=200}
  n20 = Constant() {r=-100}
  n21 = LinearInterpolate(
   A=LinearInterpolate(
    B=@n18
    Alpha=MaterialFunctionCall!(Blend Range=@n19, Start Offset=@n20) {materialFunction=/Engine/Functions/Engine_MaterialFunctions03/Particles/Distance_Blend}.Result
   )
   Alpha=MaterialFunctionCall(Blend Range=Constant() {r=600}, Start Offset=Constant() {r=-1050}) {materialFunction=&Distance_Blend}.Result
  ) {constB=0}
  n22 = CollectionParameter() {
   collection=&FortniteMaterialParameters
   parameterName=FortniteActiveTimeOfDay
  }
  n23 = Constant4Vector() {constant=<rgba 1 1 1 0>}
  DepthFade(Opacity=LinearInterpolate(A=@n21, B=Multiply(A=@n21) {constB=0.6}, Alpha=DotProduct(A=@n22, B=@n23.RGBA)) {desc="A=Night B=Day"})
  n24 = CollectionParameter() {
   collection=&FortniteMaterialParameters
   parameterName=FortniteActiveTimeOfDay
  }
  n25 = LinearInterpolate(A=Multiply(A=@n14.RGB) {constB=0.5}, B=@n14.RGB, Alpha=DotProduct(A=@n24, B=Constant4Vector() {constant=<rgba 1 1 1 0>}.RGBA)) {desc="A=Night B=Day"}
  n26 = Clamp(None=Multiply(A=Multiply(A=@n12.R, B=@n25)) {constB=2})
  n27 = OneMinus(None=@n18)
  n28 = Multiply(A=Power(Base=Clamp(None=@n27)) {constExponent=1})
  n29 = Multiply(A=Multiply(A=@n26, B=Clamp(None=@n28))) {constB=2}
  MaterialFunctionCall(DiffuseColor=Add(A=@n26, B=Clamp(None=@n29))) {materialFunction=&ConvertFromDiffSpec}
 ]
}
