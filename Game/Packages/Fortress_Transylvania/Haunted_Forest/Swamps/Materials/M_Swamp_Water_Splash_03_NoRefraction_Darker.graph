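# Translucent lit particle and Niagara sprite material for a dark swamp water splash, with no refraction.
# A panning noise texture distorts splash UVs, while dynamic parameters blend in and shape a tiled tornado-pattern breakup.
# View-space normals dotted with two fixed light vectors and sharpened by powers fake specular glints, tinted by particle color.
# Time of day from a parameter collection adjusts desaturation and opacity, with optional depth fade.
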Material /Game/Packages/Fortress_Transylvania/Haunted_Forest/Swamps/Materials/M_Swamp_Water_Splash_03_NoRefraction_Darker
graph {
 properties {
  bTangentSpaceNormal=false
  bUsedWithNiagaraSprites=true
  bUsedWithParticleSprites=true
  blendMode=BLEND_Translucent
  opacityMaskClipValue=0.999
  pixelDepthOffsetMode=PDOM_Legacy
  translucencyPass=MTP_BeforeDOF
  usedShadingModels=Default Lit
 }
 nodes [
  n1 = Panner(Coordinate=TextureCoordinate() {uTiling=0.75, vTiling=0.75}) {speedX=-0.16}
  n2 = ComponentMask(None=TextureSample(UVs=@n1) {texture=&2ndInfectedMapLinear}.R) {g=true, r=true}
  n3 = Add(A=Multiply(A=@n2) {constB=0.05}, B=TextureCoordinate())
  n4 = TextureSample(UVs=@n3) {texture=/Game/Effects/Fort_Effects/Textures/Water/T_water_splash_04}
  n5 = Multiply(A=@n4.A) {constA=1, constB=0.5}
  n6 = Clamp(None=Add(A=@n4.B) {constB=0.2})
  n7 = TextureCoordinate()
  n8 = DynamicParameter() {defaultValue=<rgba 0 0 0 1>, paramNames [Param1,Param2,VTiling,UTiling]}
  n9 = AppendVector(A=@n8.VTiling, B=@n8.VTiling)
  n10 = Add(A=Multiply(A=@n2) {constB=0.05}, B=Panner(Coordinate=Multiply(A=@n7, B=@n9)))
  n11 = Power(Base=Multiply(A=@n6, B=TextureSample(UVs=@n10) {texture=/Game/Global/Textures/Default/Tileable_Patterns/T_Tornado_Tile}.R), Exp=@n8.Param2)
  n12 = ParticleColor()
  n13 = Multiply(A=LinearInterpolate(A=@n5, B=Multiply(A=@n5, B=@n11), Alpha=@n8.Param1), B=@n12.A)
  n14 = CollectionParameter() {
   collection=&FortniteMaterialParameters
   parameterName=FortniteActiveTimeOfDay
  }
  n15 = DotProduct(A=@n14, B=Constant4Vector() {constant=<rgba 1 1 1 0>}.RGBA)
  n16 = LinearInterpolate(A=@n13, B=Multiply(A=@n13) {constB=0.5}, Alpha=@n15) {desc="A=Night B=Day"}
  LinearInterpolate(
   A=@n16
   B=DepthFade(Opacity=@n16, FadeDistance=ScalarParameter() {defaultValue=100, parameterName=FadeDistance})
   Alpha=ScalarParameter() {parameterName=UesDepthFade}
  ) {desc="A=Night B=Day"}
  MaterialFunctionCall(Emissive=Multiply(A=@n12.RGB) {constB=0.25}) {materialFunction=&MF_FX_EmissiveAdjustments}
  n17 = Add(
   A=Multiply(
    A=TextureSample(UVs=@n3) {samplerType=SAMPLERTYPE_Normal, texture=/Game/Effects/Fort_Effects/Textures/Water/T_water_splash_04_n}.RGB
    B=Constant3Vector() {constant=<rgba 3 3 1 0>}
   )
   B=TextureSample(UVs=@n1) {samplerType=SAMPLERTYPE_Normal, texture=/Game/Global/Textures/Default/Tileable_Patterns/2ndInfectedMapNormal}.RGB
  )
  n18 = Transform(None=@n17) {transformType=TRANSFORM_View}
  n19 = Multiply(
   A=DotProduct(A=Normalize(VectorInput=@n18), B=Constant3Vector() {desc=Control Light angle, constant=<rgba 0.4 -0.94 0.3 0>})
   B=Constant() {r=1}
  )
  n20 = Multiply(A=Multiply(A=Power(Base=@n19) {constExponent=20}) {constB=5}) {constB=0.5}
  n21 = DotProduct(
   A=Normalize(VectorInput=Transform(None=@n17) {transformType=TRANSFORM_View})
   B=Constant3Vector() {desc=Control Light angle, constant=<rgba 0.1 0.6 0.8 0>}
  )
  n22 = Constant() {r=1}
  n23 = Multiply(A=Power(Base=Multiply(A=@n21, B=@n22)) {constExponent=20}) {constB=5}
  n24 = Multiply(A=@n23) {constB=0.5}
  n25 = Multiply(A=LinearInterpolate(A=@n6, B=@n4.A, Alpha=Clamp(None=@n24)), B=@n4.A)
  n26 = Multiply(A=OneMinus(None=Power(Base=@n25)), B=ComponentMask(None=@n13) {r=true})
  n27 = LinearInterpolate(A=Clamp(None=Add(A=@n26) {constB=0.4}), Alpha=@n8.Param1) {constB=0}
  n28 = Clamp(None=Multiply(A=LinearInterpolate(A=@n24, B=@n23, Alpha=@n27), B=@n27)) {maxDefault=3}
  n29 = Multiply(A=Desaturation(None=@n12.RGB, Fraction=Constant() {r=0.5})) {constB=2}
  n30 = Multiply(A=LinearInterpolate(A=@n4.R, B=Multiply(A=@n4.R, B=@n11), Alpha=@n8.Param1), B=@n12.RGB)
  n31 = Add(A=Multiply(A=Clamp(None=@n20), B=@n12.RGB), B=Add(A=Multiply(A=@n28, B=@n29), B=Multiply(A=@n30, B=@n6)))
  n32 = Constant() {r=0.5}
  MaterialFunctionCall(DiffuseColor=LinearInterpolate(A=@n31, B=Desaturation(None=@n31, Fraction=@n32), Alpha=@n15) {desc="A=Night B=Day"}) {materialFunction=&ConvertFromDiffSpec}
 ]
}
