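# Translucent, two-sided water splash material for mesh particles, including Niagara, from the Haunted Forest swamp set.
# Panning noise distorts the splash color and normal textures, while a scrolling streak pattern breaks up color and opacity through dynamic parameters.
# A view-space normal dotted with a fixed vector fakes specular glints.
# Opacity combines particle alpha, optional camera-distance fade, optional height-lerp dissolve and depth fade; scalar switches also enable particle-colored emissive.
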Material /Game/Packages/Fortress_Transylvania/Haunted_Forest/Swamps/Materials/M_Water_Splash_Mesh
graph {
 properties {
  bTangentSpaceNormal=false
  bUsedWithMeshParticles=true
  bUsedWithNiagaraMeshParticles=true
  blendMode=BLEND_Translucent
  opacityMaskClipValue=0.999
  pixelDepthOffsetMode=PDOM_Legacy
  previewMesh=/Game/Packages/Fortress_Transylvania/Haunted_Forest/Swamps/Mesh/SM_water_splash_inside_01
  refractionMethod=RM_IndexOfRefraction
  translucencyLightingMode=TLM_VolumetricPerVertexNonDirectional
  twoSided=true
  usedShadingModels=Default Lit
 }
 nodes [
  n1 = Constant3Vector() {constant=<rgba 0 0 1 0>}
  n2 = Panner(Coordinate=TextureCoordinate() {uTiling=0.5, vTiling=0.5}) {speedY=-0.06}
  n3 = ComponentMask(None=TextureSample(UVs=@n2) {texture=&2ndInfectedMapLinear}.R) {g=true, r=true}
  n4 = Add(A=Multiply(A=@n3) {constB=0.02}, B=TextureCoordinate())
  n5 = Add(
   A=Multiply(
    A=TextureSample(UVs=@n4) {samplerType=SAMPLERTYPE_Normal, texture=/Game/Packages/Fortress_Transylvania/Haunted_Forest/Swamps/Textures/T_water_splash_05_n}.RGB
    B=Constant3Vector() {constant=<rgba 6 6 1 0>}
   )
   B=TextureSample(UVs=@n2) {samplerType=SAMPLERTYPE_Normal, texture=/Game/Global/Textures/Default/Tileable_Patterns/2ndInfectedMapNormal}.RGB
  )
  n6 = TextureSample(UVs=@n4) {texture=/Game/Packages/Fortress_Transylvania/Haunted_Forest/Swamps/Textures/T_water_splash_05}
  n7 = Clamp(None=Add(A=@n6.B) {constB=0.3})
  n8 = TextureCoordinate() {uTiling=3}
  n9 = DynamicParameter() {defaultValue=<rgba 0 0 0 1>, paramNames [Param1,Param2,VTiling,UTiling]}
  n10 = Add(A=Multiply(A=@n3) {constB=0.02}, B=Panner(Coordinate=Multiply(A=@n8, B=@n9.UTiling)) {speedY=-1})
  n11 = Power(Base=Multiply(A=@n7, B=TextureSample(UVs=@n10) {texture=/Game/Global/Textures/Default/Tileable_Patterns/T_Tornado_Tile}.R), Exp=@n9.Param2)
  n12 = Clamp(None=Multiply(A=@n6.A) {constB=80})
  n13 = Clamp(None=LinearInterpolate(A=@n6.A, B=Multiply(A=@n11, B=@n12), Alpha=Multiply(A=@n9.Param1) {constB=3}))
  n14 = Clamp(None=Multiply(A=LinearInterpolate(A=@n1, B=@n5, Alpha=@n13)) {constB=10})
  n15 = DepthFade() {fadeDistanceDefault=10}
  Add(A=Multiply(A=LinearInterpolate(A=@n1, B=@n14, Alpha=@n15), B=@n9.VTiling) {constB=0.1}, B=Constant() {r=1})
  n16 = Multiply(A=@n6.A) {constA=1, constB=0.5}
  n17 = ParticleColor()
  n18 = Multiply(A=LinearInterpolate(A=@n16, B=Multiply(A=@n16, B=@n11), Alpha=@n9.Param1), B=@n17.A)
  n19 = Constant() {r=100}
  n20 = Constant() {r=-130}
  n21 = LinearInterpolate(
   A=LinearInterpolate(
    B=@n18
    Alpha=MaterialFunctionCall!(Blend Range=@n19, Start Offset=@n20) {materialFunction=/Engine/Functions/Engine_MaterialFunctions03/Particles/Distance_Blend}.Result
   )
   B=@n18
   Alpha=ScalarParameter() {parameterName=RemoveCameraFade}
  )
  n22 = Add(A=TextureCoordinate() {uTiling=2, vTiling=0.5}, B=Constant2Vector() {r=0.2})
  n23 = Round(None=MaterialFunctionCall(
    Transition Phase=DynamicParameter() {parameterIndex=1, defaultValue=<rgba 0 0 0 1>, paramNames [Dissolve,Param2,VTiling,UTiling]}.Dissolve
    Height Texture=TextureSample(UVs=@n22) {texture=/Game/CharClasses/Outlander/Gadgets/AntiMaterialCharge/Materials/T_AntiMaterialCharge_Dissolve_01}.R
   ) {materialFunction=&HeightLerp}.Results)
  DepthFade(
   Opacity=LinearInterpolate(A=@n21, B=Multiply(A=@n21, B=@n23), Alpha=ScalarParameter() {parameterName=UseDissolve})
   FadeDistance=ScalarParameter() {defaultValue=20, parameterName=FadeDistance}
  ) {fadeDistanceDefault=20}
  MaterialFunctionCall() {materialFunction=&DebugTimeSine}
  MaterialFunctionCall(Emissive=LinearInterpolate(B=ParticleColor().RGB, Alpha=ScalarParameter() {parameterName=UseEmissive})) {materialFunction=&MF_FX_EmissiveAdjustments}
  n24 = MaterialFunctionCall() {materialFunction=&LinearGradient}
  n25 = OneMinus(None=@n24.VGradient)
  n26 = DotProduct(
   A=Normalize(VectorInput=Transform(None=@n5) {transformType=TRANSFORM_View})
   B=Constant3Vector() {desc=Control Light angle, constant=<rgba 0 -1 1 0>}
  )
  n27 = Constant() {r=1}
  n28 = Multiply(A=Power(Base=Multiply(A=@n26, B=@n27)) {constExponent=10}) {constB=5}
  n29 = Multiply(A=Multiply(A=Clamp(None=@n25), B=Clamp(None=@n28)) {constA=1}, B=@n6.R)
  n30 = LinearInterpolate(A=@n6.G, B=Multiply(A=@n11, B=@n6.G), Alpha=@n9.Param1)
  n31 = Add(A=Multiply(A=Clamp(None=@n29)) {constB=4}, B=Multiply(A=@n7, B=Multiply(A=@n17.RGB, B=@n30)))
  n32 = LinearInterpolate(A=@n31, B=Multiply(A=@n31) {constB=4}, Alpha=@n24.VGradient)
  n33 = ParticleColor()
  MaterialFunctionCall(DiffuseColor=LinearInterpolate(A=@n32, B=Multiply(A=@n32, B=@n33.RGB), Alpha=ScalarParameter() {parameterName=UseEmissive})) {materialFunction=&ConvertFromDiffSpec}
 ]
}
