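# Masked, two-sided subsurface material for a swamp leaf in the Fortress Transylvania Haunted Forest package, flagged for mesh particle and Niagara mesh particle use.
# One texture parameter, defaulting to t_leaf_01, supplies RGB to a ConvertFromDiffSpec function as diffuse color and to a multiply that halves it.
# The graph as listed shows no links from these nodes to the material outputs.
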Material /Game/Packages/Fortress_Transylvania/Haunted_Forest/Swamps/Materials/M_leaf_01
graph {
 properties {
  bUsedWithMeshParticles=true
  bUsedWithNiagaraMeshParticles=true
  blendMode=BLEND_Masked
  pixelDepthOffsetMode=PDOM_Legacy
  shadingModel=MSM_Subsurface
  twoSided=true
  usedShadingModels=Subsurface
 }
 nodes [
  MaterialExpressionTextureSampleParameter2D_0 = TextureSampleParameter2D() {
   parameterName=MaterialExpressionTextureSampleParameter2D_0
   texture=/Game/Packages/Fortress_Transylvania/Haunted_Forest/Swamps/Textures/t_leaf_01
  }
  MaterialFunctionCall(DiffuseColor=@MaterialExpressionTextureSampleParameter2D_0.RGB) {materialFunction=&ConvertFromDiffSpec}
  Multiply(A=@MaterialExpressionTextureSampleParameter2D_0.RGB) {constB=0.5}
 ]
}
